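# SoundCue for PastaRipper pulse rifle impacts on flesh, layering a boosted random sizzle with a random pitched-up husk impact.
# A quality switch sends its first two levels to a mix with larger variation pools and its last two to a three-plus-three mix.
# Only the smaller mix also pitch-shifts the sizzle.
# It plays in the impacts sound class, overriding attenuation and capping concurrency by stopping the oldest.
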SoundCue /RadicalWeaponsGameplay/Weapons/PulseRifleMMObj/Audio/SC_PastaRipper_ImpactFlesh
root=QualityLevel_0
nodes [
 {id=Mixer_0, children [Modulator_3,Modulator_0], properties {inputVolume [1.6,1]}}
 {id=Mixer_11, children [Random_4,Modulator_1], properties {inputVolume [1.6,1]}}
 {id=Modulator_0, children [Random_3], properties {pitchMax=1.3, pitchMin=1.2}}
 {id=Modulator_1, children [Random_0], properties {pitchMax=1.3, pitchMin=1.2}}
 {id=Modulator_3, children [Random_5], properties {pitchMax=1.3, pitchMin=1.2}}
 {id=QualityLevel_0, children [Mixer_11,Mixer_11,Mixer_0,Mixer_0]}
 {id=Random_0, children [WavePlayer_0,WavePlayer_1,WavePlayer_2,WavePlayer_3,WavePlayer_4,WavePlayer_5], properties {weights [1,1,1,1,1,1]}}
 {id=Random_3, children [WavePlayer_6,WavePlayer_7,WavePlayer_8], properties {weights [1,1,1]}}
 {id=Random_4, children [WavePlayer_21, WavePlayer_22, WavePlayer_23, WavePlayer_29], properties {weights [1,1,1,1]}}
 {id=Random_5, children [WavePlayer_30, WavePlayer_31, WavePlayer_32], properties {weights [1,1,1]}}
 {id=WavePlayer_0, properties {soundWaveAssetPtr=/RadicalWeaponsGameplay/Weapons/PulseRifleMMObj/Audio/SoundWaves/husk_impact_mix_ST_01}}
 {id=WavePlayer_1, properties {soundWaveAssetPtr=/RadicalWeaponsGameplay/Weapons/PulseRifleMMObj/Audio/SoundWaves/husk_impact_mix_ST_04}}
 {id=WavePlayer_2, properties {soundWaveAssetPtr=/RadicalWeaponsGameplay/Weapons/PulseRifleMMObj/Audio/SoundWaves/husk_impact_mix_ST_07}}
 {
  id=WavePlayer_21
  properties {soundWaveAssetPtr=/RadicalWeaponsGameplay/Weapons/PulseRifleMMObj/Audio/SoundWaves/PastaRipper_Impact_Sizzle_01}
 }
 {
  id=WavePlayer_22
  properties {soundWaveAssetPtr=/RadicalWeaponsGameplay/Weapons/PulseRifleMMObj/Audio/SoundWaves/PastaRipper_Impact_Sizzle_02}
 }
 {
  id=WavePlayer_23
  properties {soundWaveAssetPtr=/RadicalWeaponsGameplay/Weapons/PulseRifleMMObj/Audio/SoundWaves/PastaRipper_Impact_Sizzle_03}
 }
 {
  id=WavePlayer_29
  properties {soundWaveAssetPtr=/RadicalWeaponsGameplay/Weapons/PulseRifleMMObj/Audio/SoundWaves/PastaRipper_Impact_Sizzle_04}
 }
 {id=WavePlayer_3, properties {soundWaveAssetPtr=/RadicalWeaponsGameplay/Weapons/PulseRifleMMObj/Audio/SoundWaves/husk_impact_mix_ST_09}}
 {
  id=WavePlayer_30
  properties {soundWaveAssetPtr=&PastaRipper_Impact_Sizzle_01}
 }
 {
  id=WavePlayer_31
  properties {soundWaveAssetPtr=&PastaRipper_Impact_Sizzle_02}
 }
 {
  id=WavePlayer_32
  properties {soundWaveAssetPtr=&PastaRipper_Impact_Sizzle_03}
 }
 {id=WavePlayer_4, properties {soundWaveAssetPtr=/RadicalWeaponsGameplay/Weapons/PulseRifleMMObj/Audio/SoundWaves/husk_impact_mix_ST_10}}
 {id=WavePlayer_5, properties {soundWaveAssetPtr=/RadicalWeaponsGameplay/Weapons/PulseRifleMMObj/Audio/SoundWaves/husk_impact_mix_ST_13}}
 {id=WavePlayer_6, properties {soundWaveAssetPtr=&husk_impact_mix_ST_01}}
 {id=WavePlayer_7, properties {soundWaveAssetPtr=&husk_impact_mix_ST_04}}
 {id=WavePlayer_8, properties {soundWaveAssetPtr=&husk_impact_mix_ST_07}}
]
properties {
 bOverrideAttenuation=true
 bOverrideConcurrency=true
 maxDistance=3400
 soundClassObject=&Fort_Impacts
 volumeMultiplier=0.8
 attenuationOverrides {falloffDistance=3000}
 concurrencyOverrides {maxCount=5, resolutionRule=StopOldest, volumeScale=0.75}
}
