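# Sound cue for the low-ammo sound of the PastaRipper pulse rifle.
# A short randomized delay feeds a modulator that randomly varies pitch upward and boosts volume on a single low-ammo wave.
# It routes to the degraded ranged-weapon low-ammo sound class and has no distance attenuation.
# A concurrency override prevents new instances once the limit is reached rather than stopping playing ones.
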SoundCue /RadicalWeaponsGameplay/Weapons/PulseRifleMMObj/Audio/SC_PastaRipper_LowAmmo
root=Delay_1
nodes [
 {id=Delay_1, children [Modulator_1], properties {delayMax=0.08, delayMin=0.07}}
 {id=Modulator_1, children [WavePlayer_4], properties {pitchMax=1.3, pitchMin=1, volumeMax=1.8, volumeMin=1.7}}
 {
  id=WavePlayer_4
  properties {soundWaveAssetPtr=/RadicalWeaponsGameplay/Weapons/PulseRifleMMObj/Audio/SoundWaves/PastaRipper_PulseRifle_LowAmmo}
 }
]
properties {
 bOverrideConcurrency=true
 maxDistance=2097152
 soundClassObject=/Game/Sounds/FortSoundClassesAndModes/Fort_WeaponRanged_LowAmmo_Degraded_PSM
 volumeMultiplier=0.8
 attenuationOverrides {bAttenuate=false, stereoSpread=5000}
 concurrencyOverrides {resolutionRule=PreventNew}
}
