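; Assault rifle weapon Blueprint, apparently the PastaRipper pulse rifle variant, that tints its muzzle flash depending on aim-down-sights state.
; Its muzzle flash event runs the parent logic, then calls MuzzleADSCheck to pick between the ADS color and the hipfire color.
; The chosen color is converted to a vector and written to the InitialColor+ADS parameter of the spawned muzzle particle system.
; The construction script only calls the parent.
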(event EventMuzzleFlashFX
  (params (Persistent%20Fire :type Boolean))
  (|CallParentFunction Persistent%20Fire)
  (bind _return_ads_bool (CallFunction|MuzzleADSCheck))
  (Effects|Components|ParticleSystem|SetVectorParameter (Variables|Effects|Muzzle|GetSpawned%20Muzzle%20System) "InitialColor+ADS" (select _return_ads_bool (Math|Conversions|ToVector(LinearColor) (Variables|Default|GetADSColor)) (Math|Conversions|ToVector(LinearColor) (Variables|Default|GetHipfire%20Color)))))

(fn ConstructionScript
  (|CallParentFunction))
