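# Niagara beam effect for the Assault PastaRipper weapon: a short-lived smoke ribbon, likely a shot trail.
# Its single emitter lays an instantaneous burst of ribbon particles between the simulation position and the absolute endpoint User.BeamEnd_0.
# Over each particle's life, curves fade alpha and drive dynamic material parameters.
# The ribbon stays static after spawn, as beam update, curl noise, force solving and width scaling are disabled.
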NiagaraSystem /RadicalWeaponsGameplay/Weapons/PulseRifleMMObj/FX/NS_Assault_PastaRipper_Beam
dependencies {
 usedDataInterfaces [&NiagaraDataInterfaceCurve]
 usedDynamicInputs [
  &Multiply_VectorByFloat
  &Divide_Float
  /Niagara/DynamicInputs/TypeConversions/VectorLengthSafe
  &Subtract_Vector
 ]
 usedModules [
  &self:SystemSpawnScript.SystemScriptSource.SystemScriptGraph.NiagaraNodeAssignment_0.SetVariables_#1
  &SystemState
 ]
}
emitters {NE_FN_Beam_Burst_Smoke {
  emitterSpawnScript [/SetVariables(Emitter.RandomLifeTime:NiagaraFloat=&UniformRangedFloat(Minimum=0.6, Maximum=0.75)) {enabled=false, isSetParameters=true}]
  emitterUpdateScript [
   &EmitterState()
   &EmitterScalability(
    Module.bApplySpawnCountScale=false
    Module.bApplySpawnCountScaleByDistanceFraction=false
    Module.bUseMaxDistance=false
    Module.bUseMinDistance=false
   ) {enabled=false}
   /Niagara/Modules/Beams/BeamEmitterSetup(Absolute Beam End=true, Beam End:Vector3f=$User.BeamEnd_0, Beam Start:Vector3f=/Niagara/DynamicInputs/Helpers/SimulationPosition())
   &EmitterLifeCycle(MaxLoopCount=1, NextLoopDuration=0) {enabled=false}
   &SpawnBurst_Instantaneous(Spawn Count=10, Use Spawn Probability=false)
  ]
  particleSpawnScript [
   /Niagara/Modules/Beams/SpawnBeam()
   &InitializeRibbon(
    Lifetime=&RandomRangeFloat(Minimum=0.5, Maximum=0.75)
    Module.Write Color=true
    Module.Write Lifetime=true
    Module.Write Mass=false
    Module.Write Material Random=true
    Module.Write Position=false
    Module.Write RibbonTwist=false
    Module.Write RibbonWidth=true
    Position:Vector3f=/Niagara/Functions/Localspace/SimulationPosition()
    Ribbon Width=20
   )
  ]
  particleUpdateScript [
   &ParticleState()
   /Niagara/Modules/Beams/UpdateBeam() {enabled=false}
   &ScaleColor(
    Scale Alpha=&FloatFromCurve(FloatCurve=DataInterface() {propertyValues {Curve {keys [
         <at -0.0046 1.0113 RCIM_Cubic>
         {arriveTangent=-0.945, interpMode=RCIM_Cubic, leaveTangent=-0.945, tangentMode=RCTM_User, time=1.0011, value=0.002}
        ]}}})
    ScaleA=true
    ScaleRGB=true
   )
   /Niagara/Modules/Update/Forces/CurlNoiseForce(
    Noise Strength=&FloatFromCurve(FloatCurve=DataInterface() {propertyValues {Curve {keys [
         <at 0.0029 1.0113 RCIM_Cubic>
         {arriveTangent=161.8548, interpMode=RCIM_Cubic, leaveTangent=161.8275, tangentMode=RCTM_User, time=1.0011, value=69.35}
        ]}}})
    Pan Noise=false
   ) {enabled=false}
   &CurlNoiseForce(
    Noise Strength=&FloatFromCurve(FloatCurve=DataInterface() {propertyValues {Curve {keys [
         <at 0.0029 1.0113 RCIM_Cubic>
         {arriveTangent=-269.9427, interpMode=RCIM_Cubic, leaveTangent=-269.9427, tangentMode=RCTM_User, time=0.38, value=1.64}
        ]}}})
    Pan Noise=false
   ) {enabled=false, module=CurlNoiseForce001}
   /Niagara/Modules/Update/Material/DynamicMatParams(
    Module.Param0WriteEnabled=true
    Module.Param1WriteEnabled=true
    Module.Param2WriteEnabled=true
    Module.Param3WriteEnabled=false
    Param0Value:NiagaraFloat=&FloatFromCurve(FloatCurve=DataInterface() {propertyValues {Curve {keys [
         <at -0.0046 1.0113 RCIM_Cubic>
         {arriveTangent=-2.827, interpMode=RCIM_Cubic, leaveTangent=-2.827, tangentMode=RCTM_User, time=1.0011, value=0.002}
        ]}}})
    Param1Value:NiagaraFloat=&FloatFromCurve(FloatCurve=DataInterface() {propertyValues {Curve {keys [{arriveTangent=5.2814, interpMode=RCIM_Cubic, leaveTangent=5.2814, tangentMode=RCTM_User, time=-0.0046, value=0}, <at 1.0011 2 RCIM_Cubic>]}}})
    Param2Value:NiagaraFloat=&FloatFromCurve(FloatCurve=DataInterface() {propertyValues {Curve {keys [
         {arriveTangent=-7.0562, interpMode=RCIM_Cubic, leaveTangent=-7.0562, tangentMode=RCTM_User, time=-0.0046, value=1}
         <at 0.44 0.002 RCIM_Cubic>
        ]}}})
   ) {enabled=false}
   &DynamicMaterialParameters(
    Index 0 Param 1=&FloatFromCurve(FloatCurve=DataInterface() {propertyValues {Curve {keys [
         {arriveTangent=-1.6147, interpMode=RCIM_Cubic, leaveTangent=-1.6147, tangentMode=RCTM_User, time=-0.0046, value=1.0113}
         {interpMode=RCIM_Cubic, leaveTangent=-2.827, time=1.0011, value=0.002}
        ]}}})
    Index 0 Param 2=&FloatFromCurve(
     FloatCurve=DataInterface() {propertyValues {Curve {keys [{arriveTangent=5.2814, interpMode=RCIM_Cubic, leaveTangent=5.2814, tangentMode=RCTM_User, time=-0.0046, value=0}, <at 1.0011 2 RCIM_Cubic>]}}}
     Scale Curve=0.5
    )
    Index 0 Param 3=&FloatFromCurve(FloatCurve=DataInterface() {propertyValues {Curve {keys [{arriveTangent=-7.0562, interpMode=RCIM_Cubic, time=-0.0046, value=1}, <at 0.5 0.002 RCIM_Cubic>]}}})
    Module.Param0WriteEnabled=true
    Module.Param1WriteEnabled=true
    Module.Param2WriteEnabled=true
    Module.Param3WriteEnabled=true
   )
   &ScaleRibbonWidth(Ribbon Width Scale=&FloatFromCurve(FloatCurve=DataInterface() {propertyValues {Curve {keys [<at -0.0046 0.75>,<at 0.9907 1>]}}})) {enabled=false}
   &SolveForcesAndVelocity(Clamp Velocity=false, Limit Acceleration=false) {enabled=false}
  ]
  rendererClasses [&NiagaraRibbonRendererProperties]
  renderers [{rendererClass=&NiagaraRibbonRendererProperties}]
 }}
summary {
 systemName=NS_Assault_PastaRipper_Beam
 emitters [{emitterName=NE_FN_Beam_Burst_Smoke, rendererClasses [&NiagaraRibbonRendererProperties]}]
 userVariables [User.BeamEnd_0 {type=Vector3f, description="", defaultValue {struct=Vector3f, value=<3000 0 0>}}]
}
