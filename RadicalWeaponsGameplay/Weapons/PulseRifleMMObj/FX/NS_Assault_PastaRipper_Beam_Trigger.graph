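# Niagara firing effect for the PastaRipper pulse rifle, drawing tracers and fading ribbon smoke between beam start and end points.
# Fortnite weapon modules spawn particles on firing and read the User beam position arrays and MaxRange.
# Mesh tracers use distance remaps and frame-limited travel, passing start position and beam length to dynamic material parameters.
# The cyan-tinted tracer emitter is active; its near-duplicate Tracers_InfLoop is disabled.
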NiagaraSystem /RadicalWeaponsGameplay/Weapons/PulseRifleMMObj/FX/NS_Assault_PastaRipper_Beam_Trigger
dependencies {
 usedDataInterfaces [&NiagaraDataInterfaceCurve]
 usedDynamicInputs [
  &Multiply_VectorByFloat
  &Divide_Float
  /Niagara/DynamicInputs/TypeConversions/VectorLengthSafe
  &Subtract_Vector
 ]
 usedModules [
  &self:SystemSpawnScript.SystemScriptSource.SystemScriptGraph.NiagaraNodeAssignment_0.SetVariables_#1
  &SystemState
  /Game/Effects/Niagara/Modules/Weapons/FN_Setup_Weapons
 ]
}
emitters {
 NE_FN_Beam_Burst_Smoke {
  emitterSpawnScript [/SetVariables(Emitter.RandomLifeTime:NiagaraFloat=&UniformRangedFloat(Minimum=0.6, Maximum=0.75)) {enabled=false, isSetParameters=true}]
  emitterUpdateScript [
   &EmitterState()
   /Game/Effects/Niagara/Modules/Weapons/FN_Spawn_When_Firing(Age=0, MaxRange:NiagaraFloat=$User.MaxRange)
   /Niagara/Modules/Beams/BeamEmitterSetup(Absolute Beam End=true, Beam End:Vector3f=$User.BeamEnd_0, Beam Start:Vector3f=/Niagara/DynamicInputs/Helpers/SimulationPosition()) {enabled=false}
   &SpawnBurst_Instantaneous(Spawn Count=10, Use Spawn Probability=false) {enabled=false}
  ]
  particleSpawnScript [
   /Niagara/Modules/Beams/SpawnBeam() {enabled=false}
   &PartitionParticles(
    Particles Per Partition=2
    Partition Count Mode:ENiagara_InfiniteLoopDuration=::NewEnumerator1
    Partition Mode:ENiagara_PartitionMode=::NewEnumerator1
   )
   /SetVariables(Particles.Position=/Niagara/DynamicInputs/Arrays/SelectPositionFromArray(
     Array Sampling Mode:ENiagaraArraySamplingMode=::NewEnumerator1
     Direct Array Index:NiagaraInt32=$Particles.PartitionParticles.Partition
     Interpolation:NiagaraFloat=?
     Direct Array Mode:ENiagara_WrapClamp=::NewEnumerator1
     Evaluation Type:ENiagaraRandomnessEvaluation=::NewEnumerator0
     Normalized Interpolation:NiagaraBool=true
     Recalculate Random Each Loop:NiagaraBool=false
     Position Array:NiagaraDataInterfaceArrayPosition=DataInterface() {propertyValues {GpuSyncMode=SyncCpuToGpu, MaxElements=0, PositionData [<0 0 0>, <0 0 50>]}}
     Randomness Mode:ENiagaraRandomnessMode=::NewEnumerator0
     Fixed Random Seed:NiagaraBool=false
     Random Seed:NiagaraInt32=?
    )) {isSetParameters=true}
   /Game/Effects/Niagara/Modules/Weapons/FN_Initialize_Tracers(
    BeamEndPositions:NiagaraDataInterfaceArrayPosition=$User.BeamEndPositions
    BeamStartPositions:NiagaraDataInterfaceArrayPosition=$User.BeamStartPositions
    Remap By Distance Alpha=DataInterface() {propertyValues {Curve {keys [
        {arriveTangent=0.0002, interpMode=RCIM_Cubic, leaveTangent=0.0002, tangentMode=RCTM_User, time=0, value=0}
        {arriveTangent=0.0001, arriveTangentWeight=833.3334, interpMode=RCIM_Cubic, leaveTangent=0.0001, tangentMode=RCTM_User, time=2500, value=0.35}
        <at 15000 0.8>
        {arriveTangentWeight=6666.667, interpMode=RCIM_Cubic, tangentMode=RCTM_User, time=35000, value=1}
       ]}}}
    Remap By LODDistance=DataInterface() {propertyValues {Curve {keys [<at 500 0>,<at 2500 1>]}}}
   )
   &InitializeRibbon(
    Lifetime=&RandomRangeFloat(Minimum=0.5, Maximum=0.75)
    Module.Write Color=true
    Module.Write Lifetime=true
    Module.Write Mass=false
    Module.Write Material Random=true
    Module.Write Position=false
    Module.Write RibbonTwist=false
    Module.Write RibbonWidth=true
    Position:Vector3f=/Niagara/Functions/Localspace/SimulationPosition()
    Ribbon Width=20
   )
  ]
  particleUpdateScript [
   &ParticleState()
   &ScaleColor(
    Scale Alpha=&FloatFromCurve(FloatCurve=DataInterface() {propertyValues {Curve {keys [
         <at -0.0046 1.0113 RCIM_Cubic>
         {arriveTangent=-0.945, interpMode=RCIM_Cubic, leaveTangent=-0.945, tangentMode=RCTM_User, time=1.0011, value=0.002}
        ]}}})
    ScaleA=true
    ScaleRGB=true
   )
   &DynamicMaterialParameters(
    Index 0 Param 1=&FloatFromCurve(FloatCurve=DataInterface() {propertyValues {Curve {keys [
         {arriveTangent=-1.6147, interpMode=RCIM_Cubic, leaveTangent=-1.6147, tangentMode=RCTM_User, time=-0.0046, value=1.0113}
         {interpMode=RCIM_Cubic, leaveTangent=-2.827, time=1.0011, value=0.002}
        ]}}})
    Index 0 Param 2=&FloatFromCurve(
     FloatCurve=DataInterface() {propertyValues {Curve {keys [{arriveTangent=5.2814, interpMode=RCIM_Cubic, leaveTangent=5.2814, tangentMode=RCTM_User, time=-0.0046, value=0}, <at 1.0011 2 RCIM_Cubic>]}}}
     Scale Curve=0.5
    )
    Index 0 Param 3=&FloatFromCurve(FloatCurve=DataInterface() {propertyValues {Curve {keys [{arriveTangent=-7.0562, interpMode=RCIM_Cubic, time=-0.0046, value=1}, <at 0.5 0.002 RCIM_Cubic>]}}})
    Module.Param0WriteEnabled=true
    Module.Param1WriteEnabled=true
    Module.Param2WriteEnabled=true
    Module.Param3WriteEnabled=true
   )
  ]
  rendererClasses [&NiagaraRibbonRendererProperties]
  renderers [{rendererClass=&NiagaraRibbonRendererProperties}]
 }
 Tracers_InfLoop {
  bEnabled=false
  emitterSpawnScript [/SetVariables(Emitter.DebugTimeMultiplier=1) {isSetParameters=true, module=SetVariables_#2}]
  emitterUpdateScript [
   &EmitterState()
   &FN_Spawn_When_Firing(Age=0, MaxRange:NiagaraFloat=$User.MaxRange)
  ]
  particleSpawnScript [
   &FN_Initialize_Tracers(
    BeamEndPositions:NiagaraDataInterfaceArrayPosition=$User.BeamEndPositions
    BeamStartPositions:NiagaraDataInterfaceArrayPosition=$User.BeamStartPositions
    Remap By Distance Alpha=DataInterface() {propertyValues {Curve {keys [
        {arriveTangent=0.0002, interpMode=RCIM_Cubic, leaveTangent=0.0002, tangentMode=RCTM_User, time=0, value=0}
        {arriveTangent=0.0001, arriveTangentWeight=833.3334, interpMode=RCIM_Cubic, leaveTangent=0.0001, tangentMode=RCTM_User, time=2500, value=0.35}
        <at 15000 0.8>
        {arriveTangentWeight=6666.667, interpMode=RCIM_Cubic, tangentMode=RCTM_User, time=35000, value=1}
       ]}}}
    Remap By LODDistance=DataInterface() {propertyValues {Curve {keys [<at 500 0>,<at 2500 1>]}}}
   )
   &InitializeParticle(
    Color=<rgba 1 0.3543 0.026 0.9>
    Lifetime=&Multiply_Float(
     A=&Multiply_Float(
      A=&ClampFloat(Float=$Particles.RemapByDistance, Min=0.5)
      B=&RandomRangeFloat(Minimum=0.7, Maximum=0.8)
     )
     B=$Emitter.DebugTimeMultiplier
    )
    Mesh Scale Mode=::NewEnumerator3
    Position=$Particles.BeamStartPosition
    Position Mode=::NewEnumerator1
    UsePositionOffset=false
    Write Material Random=true
   )
   &DynamicMaterialParameters(
    Index 0 Param 4=&Multiply_Float(A=$Particles.BeamTile, B=&Lerp_Float(A=0.8, B=1.35, Alpha=$Particles.RemapByLODDistance))
    Index 1 Param 1=&RandomRangeFloat(Minimum=0.85, Maximum=1.05)
    Index 1 Param 2=&RandomRangeFloat(Minimum=0, Maximum=1)
    Module.Param0WriteEnabled=false
    Module.Param0WriteEnabled1=false
    Module.Param1WriteEnabled=false
    Module.Param1WriteEnabled1=true
    Module.Param2WriteEnabled=false
    Module.Param2WriteEnabled1=false
    Module.Param3WriteEnabled=true
    Module.Param3WriteEnabled1=false
    Write Parameter Index 1=true
   ) {module=DynamicMaterialParameters001}
   /Niagara/Modules/Update/Orientation/V2/OrientMeshToVector(Look At Direction:Vector3f=$Particles.FireDirection, Override Quat=false, UseOptionalSideAxis=false, UseOptionalUpAxis=false)
   /SetVariables(Particles.BulletDynParams:Vector4f=<1 1 1 1>) {isSetParameters=true, module=SetVariables_#3}
   /SetVariables(Particles.BulletColor:LinearColor=<rgba 1 1 1 0.3>) {isSetParameters=true, module=SetVariables_#4}
  ]
  particleUpdateScript [
   &ParticleState()
   /Niagara/Modules/Update/Utility/FrameCounter()
   /Game/Effects/Niagara/Modules/Weapons/FN_Fit_Tracer_Travel(DurationMax=0.5, DurationMin=0.1, MaxFrames=7)
   &DynamicMaterialParameters(
    Index 0 Param 1=&FloatFromCurve(FloatCurve=DataInterface() {propertyValues {Curve {keys [
         {arriveTangent=-0.4135, interpMode=RCIM_Cubic, leaveTangent=-0.4135, tangentMode=RCTM_User, time=0, value=1}
         {arriveTangent=-1.3177, interpMode=RCIM_Cubic, leaveTangent=-1.3177, tangentMode=RCTM_User, time=1, value=0}
        ]}}})
    Index 0 Param 2=&FloatFromCurve(
     FloatCurve=DataInterface() {propertyValues {Curve {keys [
         {arriveTangent=5.2627, interpMode=RCIM_Cubic, leaveTangent=5.2627, tangentMode=RCTM_User, time=0, value=0}
         {arriveTangent=0.8931, interpMode=RCIM_Cubic, leaveTangent=0.8931, tangentMode=RCTM_User, time=0.2287, value=0.7}
         {arriveTangentWeight=0.2571, interpMode=RCIM_Cubic, tangentMode=RCTM_User, time=1, value=1}
        ]}}}
     Scale Curve=&Multiply_Float(
      A=&Lerp_Float(A=0.5, B=1, Alpha=$Particles.RemapByDistance)
      B=&RandomRangeFloat(Minimum=-10, Maximum=-8)
     )
    )
    Index 0 Param 3=&FloatFromCurve(
     FloatCurve=DataInterface() {propertyValues {Curve {keys [
         {arriveTangent=-0.9882, time=0.2, value=1}
         {arriveTangent=-1.2626, interpMode=RCIM_Cubic, leaveTangent=-1.2626, tangentMode=RCTM_User, time=0.5, value=0.6}
         {arriveTangent=-10.4271, interpMode=RCIM_Cubic, leaveTangent=-10.4271, tangentMode=RCTM_User, time=0.6, value=0}
        ]}}}
     CurveIndex=$Particles.FrameLimitedAge
    )
    Index 0 Param 4=&Multiply_Float(
     A=&MakeFloatFromVector(Vector=$Particles.Scale, Channel=::NewEnumerator0)
     B=&Lerp_Float(A=0.8, B=1.35, Alpha=$Particles.RemapByLODDistance)
    )
    Index 1 Param 1=&FloatFromCurve(
     FloatCurve=DataInterface() {propertyValues {Curve {keys [
         {arriveTangent=0.0005, interpMode=RCIM_Cubic, leaveTangent=0.0005, tangentMode=RCTM_User, time=0, value=0.7}
         {arriveTangent=0.9056, interpMode=RCIM_Cubic, leaveTangent=0.9056, tangentMode=RCTM_User, time=0.5, value=0.87}
         {arriveTangent=0.8994, interpMode=RCIM_Cubic, leaveTangent=0.8994, tangentMode=RCTM_User, time=0.5714, value=1.214}
         {arriveTangent=0.8128, interpMode=RCIM_Cubic, leaveTangent=0.8128, tangentMode=RCTM_User, time=1, value=1.5}
        ]}}}
     Scale Curve=&Lerp_Float(A=0.5, B=1, Alpha=$StackContext.RemapByLODDistance)
    )
    Index 3 Param 1=&MakeFloatFromVector(
     Vector=&ConvertPositionToVector(Input Position=$Particles.BeamStartPosition, Conversion Operation:ENiagara_LWCConvertPosToVec=::NewEnumerator0)
     Channel=::NewEnumerator0
    )
    Index 3 Param 2=&MakeFloatFromVector(
     Vector=&ConvertPositionToVector(Input Position=$Particles.BeamStartPosition, Conversion Operation:ENiagara_LWCConvertPosToVec=::NewEnumerator0)
     Channel=::NewEnumerator1
    )
    Index 3 Param 3=&MakeFloatFromVector(
     Vector=&ConvertPositionToVector(Input Position=$Particles.BeamStartPosition, Conversion Operation:ENiagara_LWCConvertPosToVec=::NewEnumerator0)
     Channel=::NewEnumerator2
    )
    Index 3 Param 4=&DistanceBetweenPositions(A=$Particles.BeamStartPosition, B=$Particles.BeamEndPosition)
    Module.Param0WriteEnabled=true
    Module.Param0WriteEnabled1=true
    Module.Param0WriteEnabled3=true
    Module.Param1WriteEnabled=true
    Module.Param1WriteEnabled1=false
    Module.Param1WriteEnabled3=true
    Module.Param2WriteEnabled=true
    Module.Param2WriteEnabled1=false
    Module.Param2WriteEnabled3=true
    Module.Param3WriteEnabled=false
    Module.Param3WriteEnabled1=false
    Module.Param3WriteEnabled3=true
    Write Parameter Index 1=true
    Write Parameter Index 3=true
   )
   /Game/Effects/Niagara/Modules/Weapons/FN_Update_Bullet_Travel(
    Bullet Size Factor=DataInterface() {propertyValues {Curve {keys [<at 0 0>,<at 0.05 1>]}}}
    Bullet Size Far:Vector3f=<4 4 4>
    Bullet Size Near:Vector3f=<2 2 2>
    Tracer Mesh Size Alpha=DataInterface() {propertyValues {Curve {keys [
        {interpMode=RCIM_Cubic, tangentMode=RCTM_Break, time=0.65, value=0}
        {arriveTangent=9.0989, arriveTangentWeight=0.1167, interpMode=RCIM_Cubic, tangentMode=RCTM_Break, time=1, value=1}
       ]}}}
   )
   &DynamicMaterialParameters(
    Index 0 Param 1=&FloatFromCurve(FloatCurve=DataInterface() {propertyValues {Curve {keys [
         {arriveTangent=-1.6147, interpMode=RCIM_Cubic, leaveTangent=-1.6147, tangentMode=RCTM_User, time=-0.0046, value=1.0113}
         {interpMode=RCIM_Cubic, leaveTangent=-2.827, time=1.0011, value=0.002}
        ]}}})
    Index 0 Param 2=&FloatFromCurve(
     FloatCurve=DataInterface() {propertyValues {Curve {keys [{arriveTangent=5.2814, interpMode=RCIM_Cubic, leaveTangent=5.2814, tangentMode=RCTM_User, time=-0.0046, value=0}, <at 1.0011 2 RCIM_Cubic>]}}}
     Scale Curve=0.5
    )
    Index 0 Param 3=&FloatFromCurve(FloatCurve=DataInterface() {propertyValues {Curve {keys [{arriveTangent=-7.0562, interpMode=RCIM_Cubic, time=-0.0046, value=1}, <at 0.5 0.002 RCIM_Cubic>]}}})
    Module.Param0WriteEnabled=true
    Module.Param1WriteEnabled=true
    Module.Param2WriteEnabled=true
    Module.Param3WriteEnabled=true
   ) {enabled=false, module=DynamicMaterialParameters002}
   /SetVariables(Particles.DebugScale:Vector3f=<5 5 5>) {enabled=false, isSetParameters=true, module=SetVariables_#5}
  ]
  rendererClasses [&NiagaraMeshRendererProperties]
  renderers [
   {rendererClass=&NiagaraMeshRendererProperties, rendererIndex=0}
   {rendererClass=&NiagaraMeshRendererProperties, rendererIndex=1}
  ]
 }
 Tracers_InfLoop001 {
  emitterSpawnScript [/SetVariables(Emitter.DebugTimeMultiplier=1) {isSetParameters=true, module=SetVariables_#2}]
  emitterUpdateScript [
   &EmitterState()
   &FN_Spawn_When_Firing(Age=0, MaxRange:NiagaraFloat=$User.MaxRange)
  ]
  particleSpawnScript [
   &FN_Initialize_Tracers(
    BeamEndPositions:NiagaraDataInterfaceArrayPosition=$User.BeamEndPositions
    BeamStartPositions:NiagaraDataInterfaceArrayPosition=$User.BeamStartPositions
    Remap By Distance Alpha=DataInterface() {propertyValues {Curve {keys [
        {arriveTangent=0.0002, interpMode=RCIM_Cubic, leaveTangent=0.0002, tangentMode=RCTM_User, time=0, value=0}
        {arriveTangent=0.0001, arriveTangentWeight=833.3334, interpMode=RCIM_Cubic, leaveTangent=0.0001, tangentMode=RCTM_User, time=2500, value=0.35}
        <at 15000 0.8>
        {arriveTangentWeight=6666.667, interpMode=RCIM_Cubic, tangentMode=RCTM_User, time=35000, value=1}
       ]}}}
    Remap By LODDistance=DataInterface() {propertyValues {Curve {keys [<at 500 0>,<at 2500 1>]}}}
   )
   &InitializeParticle(
    Color=<rgba 1 0.3543 0.026 0.9>
    Lifetime=&Multiply_Float(
     A=&Multiply_Float(
      A=&ClampFloat(Float=$Particles.RemapByDistance, Min=0.5)
      B=&RandomRangeFloat(Minimum=0.7, Maximum=0.8)
     )
     B=$Emitter.DebugTimeMultiplier
    )
    Mesh Scale Mode=::NewEnumerator3
    Position=$Particles.BeamStartPosition
    Position Mode=::NewEnumerator1
    UsePositionOffset=false
    Write Material Random=true
   )
   &DynamicMaterialParameters(
    Index 0 Param 4=&Multiply_Float(A=$Particles.BeamTile, B=&Lerp_Float(A=0.8, B=1.35, Alpha=$Particles.RemapByLODDistance))
    Index 1 Param 1=&RandomRangeFloat(Minimum=0.85, Maximum=1.05)
    Index 1 Param 2=&RandomRangeFloat(Minimum=0, Maximum=1)
    Module.Param0WriteEnabled=false
    Module.Param0WriteEnabled1=false
    Module.Param1WriteEnabled=false
    Module.Param1WriteEnabled1=true
    Module.Param2WriteEnabled=false
    Module.Param2WriteEnabled1=false
    Module.Param3WriteEnabled=true
    Module.Param3WriteEnabled1=false
    Write Parameter Index 1=true
   ) {module=DynamicMaterialParameters001}
   &OrientMeshToVector(Look At Direction:Vector3f=$Particles.FireDirection, Override Quat=false, UseOptionalSideAxis=false, UseOptionalUpAxis=false)
   /SetVariables(Particles.BulletDynParams:Vector4f=<1 1 1 1>) {isSetParameters=true, module=SetVariables_#3}
   /SetVariables() {isSetParameters=true, module=SetVariables_#4}
  ]
  particleUpdateScript [
   &ParticleState()
   &FrameCounter()
   &FN_Fit_Tracer_Travel(DurationMax=0.5, DurationMin=0.1, MaxFrames=8)
   &DynamicMaterialParameters(
    Index 0 Param 1=&FloatFromCurve(FloatCurve=DataInterface() {propertyValues {Curve {keys [
         {arriveTangent=-0.4135, interpMode=RCIM_Cubic, leaveTangent=-0.4135, tangentMode=RCTM_User, time=0, value=1}
         {arriveTangent=-1.3177, interpMode=RCIM_Cubic, leaveTangent=-1.3177, tangentMode=RCTM_User, time=1, value=0}
        ]}}})
    Index 0 Param 2=&FloatFromCurve(
     FloatCurve=DataInterface() {propertyValues {Curve {keys [
         {arriveTangent=5.2627, interpMode=RCIM_Cubic, leaveTangent=5.2627, tangentMode=RCTM_User, time=0, value=0}
         {arriveTangent=0.8931, interpMode=RCIM_Cubic, leaveTangent=0.8931, tangentMode=RCTM_User, time=0.2287, value=0.7}
         {arriveTangentWeight=0.2571, interpMode=RCIM_Cubic, tangentMode=RCTM_User, time=1, value=1}
        ]}}}
     Scale Curve=&Multiply_Float(
      A=&Lerp_Float(A=0.5, B=1, Alpha=$Particles.RemapByDistance)
      B=&RandomRangeFloat(Minimum=-10, Maximum=-8)
     )
    )
    Index 0 Param 3=&FloatFromCurve(
     FloatCurve=DataInterface() {propertyValues {Curve {keys [
         {arriveTangent=-0.9882, time=0.2, value=1}
         {arriveTangent=-1.2626, interpMode=RCIM_Cubic, leaveTangent=-1.2626, tangentMode=RCTM_User, time=0.5, value=0.6}
         {arriveTangent=-10.4271, interpMode=RCIM_Cubic, leaveTangent=-10.4271, tangentMode=RCTM_User, time=0.6, value=0}
        ]}}}
     CurveIndex=$Particles.FrameLimitedAge
    )
    Index 0 Param 4=&Multiply_Float(
     A=&MakeFloatFromVector(Vector=$Particles.Scale, Channel=::NewEnumerator0)
     B=&Lerp_Float(A=0.8, B=1.35, Alpha=$Particles.RemapByLODDistance)
    )
    Index 1 Param 1=&FloatFromCurve(
     FloatCurve=DataInterface() {propertyValues {Curve {keys [
         {arriveTangent=-0.8182, interpMode=RCIM_Cubic, leaveTangent=-0.8182, tangentMode=RCTM_User, time=0, value=0.7}
         {arriveTangent=2.4059, interpMode=RCIM_Cubic, leaveTangent=9.91, tangentMode=RCTM_Break, time=0.5, value=0.87}
         {arriveTangent=0.8994, interpMode=RCIM_Cubic, leaveTangent=0.8994, tangentMode=RCTM_User, time=0.5714, value=1.214}
         {arriveTangent=0.8128, interpMode=RCIM_Cubic, leaveTangent=0.8128, tangentMode=RCTM_User, time=1, value=1.5}
        ]}}}
     Scale Curve=&Lerp_Float(A=0.35, B=0.6, Alpha=$StackContext.RemapByLODDistance)
    )
    Index 3 Param 1=&MakeFloatFromVector(
     Vector=&ConvertPositionToVector(Input Position=$Particles.BeamStartPosition, Conversion Operation:ENiagara_LWCConvertPosToVec=::NewEnumerator0)
     Channel=::NewEnumerator0
    )
    Index 3 Param 2=&MakeFloatFromVector(
     Vector=&ConvertPositionToVector(Input Position=$Particles.BeamStartPosition, Conversion Operation:ENiagara_LWCConvertPosToVec=::NewEnumerator0)
     Channel=::NewEnumerator1
    )
    Index 3 Param 3=&MakeFloatFromVector(
     Vector=&ConvertPositionToVector(Input Position=$Particles.BeamStartPosition, Conversion Operation:ENiagara_LWCConvertPosToVec=::NewEnumerator0)
     Channel=::NewEnumerator2
    )
    Index 3 Param 4=&DistanceBetweenPositions(A=$Particles.BeamStartPosition, B=$Particles.BeamEndPosition)
    Module.Param0WriteEnabled=true
    Module.Param0WriteEnabled1=true
    Module.Param0WriteEnabled3=true
    Module.Param1WriteEnabled=true
    Module.Param1WriteEnabled1=false
    Module.Param1WriteEnabled3=true
    Module.Param2WriteEnabled=true
    Module.Param2WriteEnabled1=false
    Module.Param2WriteEnabled3=true
    Module.Param3WriteEnabled=false
    Module.Param3WriteEnabled1=false
    Module.Param3WriteEnabled3=true
    Write Parameter Index 1=true
    Write Parameter Index 3=true
   )
   &FN_Update_Bullet_Travel(
    Bullet Size Factor=DataInterface() {propertyValues {Curve {keys [<at 0 0>,<at 0.05 1>]}}}
    Tracer Mesh Size Alpha=DataInterface() {propertyValues {Curve {keys [
        {interpMode=RCIM_Cubic, tangentMode=RCTM_Break, time=0.65, value=0}
        {arriveTangent=9.0989, arriveTangentWeight=0.1167, interpMode=RCIM_Cubic, tangentMode=RCTM_Break, time=1, value=1}
       ]}}}
   )
   /SetVariables(Particles.DebugScale:Vector3f=<5 5 5>) {enabled=false, isSetParameters=true, module=SetVariables_#5}
   /SetVariables(Particles.BColor:LinearColor=<rgba 0 10.1876 15 1>) {isSetParameters=true}
  ]
  rendererClasses [&NiagaraMeshRendererProperties]
  renderers [
   {rendererClass=&NiagaraMeshRendererProperties, rendererIndex=0}
   {rendererClass=&NiagaraMeshRendererProperties, rendererIndex=1}
  ]
 }
}
summary {
 systemName=NS_Assault_PastaRipper_Beam_Trigger
 emitters [
  {emitterName=NE_FN_Beam_Burst_Smoke, rendererClasses [&NiagaraRibbonRendererProperties]}
  {bEnabled=false, emitterName=Tracers_InfLoop, rendererClasses [&NiagaraMeshRendererProperties]}
  {emitterName=Tracers_InfLoop001, rendererClasses [&NiagaraMeshRendererProperties]}
 ]
 userVariables [
  User.BeamEnd_0 {type=Vector3f, description="", defaultValue {struct=Vector3f, value=<0 0 0>}}
  User.BeamEndPositions {
   type=NiagaraDataInterfaceArrayPosition
   description=""
   defaultValue {
    struct=NiagaraExt_VariableValue_DataInterface
    value {
     dataInterface=&self:NiagaraDataInterfaceArrayPosition_0
     dataInterfaceClass=&NiagaraDataInterfaceArrayPosition
    }
   }
  }
  User.BeamStartPositions {
   type=NiagaraDataInterfaceArrayPosition
   description=""
   defaultValue {
    struct=NiagaraExt_VariableValue_DataInterface
    value {
     dataInterface=&self:NiagaraDataInterfaceArrayPosition_1
     dataInterfaceClass=&NiagaraDataInterfaceArrayPosition
    }
   }
  }
  User.MaxRange {type=NiagaraFloat, defaultValue=27000, description=""}
  User.Trigger {type=NiagaraBool}
 ]
}
